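# Fallback translucent, unlit overlay material that draws screen-space toon outlines on anime characters.
# Edges come from four-neighbor scene depth and optional roughness differences, with distance- and resolution-adjusted sampling.
# Opacity combines these edges with a depth-overlap mask and optional subsurface-buffer mask, which varies by shading path and quality.
# Color passes through time-of-day and LUT functions; normal-offset WPO, cloth-sim and character PostFX functions complete the output.
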Material /Game/Global/Materials/Character_Parent_2023/Anime_Parent_2023/M_AnimeOutline_Overlap_2023_Fallback
graph {
 properties {
  bAutomaticallySetUsageInEditor=false
  bEnableResponsiveAA=true
  bUseMaterialAttributes=true
  bUsedWithClothing=true
  bUsedWithMeshDeformer=true
  bUsedWithMorphTargets=true
  bUsedWithNiagaraMeshParticles=true
  bUsedWithSkeletalMesh=true
  blendMode=BLEND_Translucent
  pixelDepthOffsetMode=PDOM_Legacy
  shadingModel=MSM_Unlit
  twoSided=true
  usedShadingModels=Unlit
 }
 nodes [
  Composite() {subgraphName=Construct Base Inputs}
  n1 = NamedRerouteUsage()
  n2 = NamedRerouteUsage()
  n3 = NamedRerouteUsage()
  n4 = Constant()
  MaterialExpressionReroute_8 = Reroute(None=@n4)
  n5 = PreSkinnedPosition()
  n6 = ShaderStageSwitch(PixelShader=VertexInterpolator(VS=@n5).PS, VertexShader=@n5)
  MaterialExpressionReroute_9 = Reroute(None=@n6)
  n7 = MaterialFunctionCall!(Material) {materialFunction=/Game/Packages/Fortress_SharedMaterials/Base_Material_Functions/CharacterBrightnessandScale_Balance}
  MaterialExpressionReroute_0 = Reroute(None=@n7.Distance Blend)
  n8 = Fresnel() {baseReflectFraction=0, exponent=1}
  MaterialExpressionReroute_6 = Reroute(None=@n8)
  MaterialFunctionCall(
   Material=MaterialFunctionCall!(
    Material=SetMaterialAttributes!(MaterialAttributes, Emissive Color=@n1, Opacity=@n2, World Position Offset=@n3) {attributeSetTypes [#1, #2, #3]}
    SkinFX_Mask
   ) {materialFunction=/Game/Global/MaterialFunctions/MF_ClothSimWPO_v2}.Result
   M=@MaterialExpressionReroute_8.M
   Pre Skinned Local Position=@MaterialExpressionReroute_9.Pre Skinned Local Position
   Distance Blend=@MaterialExpressionReroute_0.Distance Blend
   Fresnel=@MaterialExpressionReroute_6.Fresnel
  ) {materialFunction=&MF_CharacterParent_2023_PostFX}
  n9 = SceneTexture() {sceneTextureId=PPI_SubsurfaceColor}
  n10 = Constant()
  n11 = If(
   A=Min(A=ComponentMask(None=@n9.Color) {g=true}, B=ComponentMask(None=@n9.Color) {b=true})
   B=SquareRoot(None=ScalarParameter() {defaultValue=0.001, group=ToonOutline_DontTouch, parameterName=Subsurface_Mask})
   "A > B"=@n10
   "A == B"=Constant() {r=1}
   "A < B"=@n10
  ) {equalsThreshold=0.0005}
  n12 = Floor(None=SceneTexture() {sceneTextureId=PPI_Opacity}.Color) {desc="for effect quality low (SSColor is disabled)"}
  n13 = SceneTexture() {sceneTextureId=PPI_ShadingModelColor}
  n14 = Constant()
  n15 = StaticSwitchParameter(
   True=If(
    A=ComponentMask(None=@n13.Color) {b=true, g=true, r=true}
    B=Constant3Vector() {constant=<rgba 0.7 1 1 0>}
    "A > B"=@n14
    "A == B"=Constant() {r=1}
    "A < B"=@n14
   ) {equalsThreshold=0.001}
   False=Constant() {desc=for PS4 and XO, r=1}
  ) {group=FXSupport, parameterName=UseInNiagaraFX, sortPriority=1}
  n16 = Constant() {r=1}
  NamedRerouteDeclaration(None=ShadingPathSwitch(Default=@n11, Deferred=QualitySwitch(Default=@n11, Low=@n12, High=@n15, Medium=@n15, Epic=@n11), Forward=@n16, Mobile=@n16)) {name=Subsurface Mask, nodeColor=<rgba 0.5469 0 1 1>}
  n17 = TextureSampleParameter2D() {
   group=ToonOutline
   parameterName=OutlineMaskTexture
   samplerType=SAMPLERTYPE_Masks
   sortPriority=2
   texture=&T_BlackMask
  }
  n18 = OneMinus(None=@n17.R)
  n19 = ScalarParameter() {defaultValue=16, group=1 - ToonOutline, parameterName=SampleBias_Near, sliderMax=16, sortPriority=15}
  n20 = ScalarParameter() {group=1 - ToonOutline, parameterName=SampleBias_Far, sliderMax=16, sortPriority=16}
  n21 = NamedRerouteUsage()
  n22 = Add(A=Multiply(A=NamedRerouteUsage(), B=LinearInterpolate(A=@n19, B=@n20, Alpha=@n21)), B=SceneDepth().R)
  n23 = ComponentMask(None=ScreenPosition().ViewportUV) {g=true, r=true}
  n24 = Divide(A=ViewSize(), B=Constant() {r=320})
  n25 = Constant() {r=2}
  n26 = ScalarParameter() {
   defaultValue=2
   desc="! Dont touch this for BR or FE tweaks !\r\nThis is only ultra high resolutions for cinematic renderings"
   group=ToonOutline_DontTouch
   parameterName=ResolutionCompensationAmount
   sliderMax=4
   sliderMin=2
   sortPriority=33
  }
  n27 = Clamp(None=Ceil(None=Subtract(A=@n24, B=@n25) {constB=0}), Max=@n26) {desc="if resolution is bigger than 960 =2", maxDefault=2, minDefault=1}
  n28 = ScalarParameter() {defaultValue=1.1, group=1 - ToonOutline, parameterName=OutlineThickness_Near, sliderMax=4, sliderMin=0.5, sortPriority=10}
  n29 = ScalarParameter() {defaultValue=0.5, group=1 - ToonOutline, parameterName=OutlineThickness_Far, sliderMax=1, sliderMin=0.5, sortPriority=11}
  n30 = MaterialFunctionCall!(
   Fade Offset=ScalarParameter() {defaultValue=50, group=1 - ToonOutline, parameterName=Outline_FadeBias, sliderMax=200, sortPriority=13}
   Fade Distance=ScalarParameter() {defaultValue=1000, group=1 - ToonOutline, parameterName=Outline_FadeDistance, sliderMax=1000, sortPriority=12}
  ) {materialFunction=/Game/Global/Materials/Character_Parent_2023/Anime_Parent_2023/MaterialFunctions/MFC_DistanceFade_FOV_compensation}
  n31 = Divide(A=Constant() {r=960}, B=ComponentMask(None=ViewSize()) {r=true})
  n32 = Max(
   A=LinearInterpolate(A=@n28, B=@n29, Alpha=NamedRerouteDeclaration(None=@n30.Result) {name=FadeDistance, nodeColor=<rgba 1 0.5773 0.4262 1>})
   B=Saturate(None=Floor(None=@n31)) {desc=if resolution is smaller than 960H}
  ) {desc="If res smaller than target, calmp min value to 1"}
  n33 = ViewProperty() {property=MEVP_ResolutionFraction}
  n34 = Multiply(A=SceneTexelSize(), B=Multiply(A=Multiply(A=@n27, B=@n32) {constA=2}, B=ComponentMask(None=@n33.Property) {r=true}))
  n35 = Constant()
  n36 = AppendVector(A=ComponentMask(None=@n34) {desc=x, r=true}, B=@n35)
  n37 = Add(A=@n23, B=@n36) {desc=x+}
  n38 = SceneDepth(UVs=@n37)
  n39 = AppendVector(A=@n35, B=ComponentMask(None=@n34) {desc=y, g=true}) {desc=y+}
  n40 = Add(A=@n23, B=@n39) {desc=y+}
  n41 = SceneDepth(UVs=@n40)
  n42 = Constant() {r=-1}
  n43 = Add(A=@n23, B=Multiply(A=@n36, B=@n42)) {desc=x-}
  n44 = SceneDepth(UVs=@n43)
  n45 = Add(A=@n23, B=Multiply(A=@n42, B=@n39)) {desc=y-}
  n46 = SceneDepth(UVs=@n45)
  n47 = Add(A=Add(A=Subtract(A=@n22, B=@n38.R), B=Subtract(A=@n22, B=@n41.R)), B=Add(A=Subtract(A=@n22, B=@n44.R), B=Subtract(A=@n22, B=@n46.R)))
  n48 = ScalarParameter() {defaultValue=-16, group=1 - ToonOutline, parameterName=Divisor_Far, sliderMax=16, sliderMin=-16, sortPriority=18}
  n49 = LinearInterpolate(
   A=ScalarParameter() {defaultValue=-1, group=1 - ToonOutline, parameterName=Divisor_Near, sliderMax=1, sliderMin=-1, sortPriority=17}
   B=@n48
   Alpha=NamedRerouteUsage()
  )
  n50 = Divide(A=AppendVector(A=@n47, B=@n47), B=AppendVector(A=@n49, B=Multiply(A=@n49) {constB=-1}))
  n51 = Multiply(
   A=OneMinus(None=Saturate(None=@n50))
   B=ScalarParameter() {defaultValue=4, group=1 - ToonOutline, parameterName=Depth_Scale, sliderMax=8, sortPriority=19}
  )
  n52 = MaterialFunctionCall(Float2=OneMinus(None=Saturate(None=@n51))) {materialFunction=&BreakOutFloat2Components}
  n53 = Max(A=@n52.R, B=Multiply(A=@n52.G, B=Constant() {desc=Divisior}))
  n54 = StaticSwitchParameter(True=Multiply(A=@n18, B=@n53), False=@n53) {
   desc=Exclude a section of the mesh from having Distance Outlines based on a Texture
   group=1 - ToonOutline
   parameterName=UseMaskTextureForDistanceOutlines
   sortPriority=2
  }
  NamedRerouteDeclaration(None=@n54) {name=DistanceOutlines, nodeColor=<rgba 0.2188 0 1 1>}
  n55 = ScalarParameter() {
   desc=a tiny value like 0.01 can improve roughness compression issues.
   group=2 - ToonOutline_Roughness
   parameterName=Roughness_SampleBias
   sliderMax=1
   sortPriority=10
  }
  n56 = Add(A=@n55, B=ComponentMask(None=SceneTexture() {sceneTextureId=PPI_Roughness}.Color) {r=true})
  n57 = ComponentMask(None=SceneTexture(UVs=@n37) {sceneTextureId=PPI_Roughness}.Color) {r=true}
  n58 = ComponentMask(None=SceneTexture(UVs=@n40) {sceneTextureId=PPI_Roughness}.Color) {r=true}
  n59 = ComponentMask(None=SceneTexture(UVs=@n43) {sceneTextureId=PPI_Roughness}.Color) {r=true}
  n60 = ComponentMask(None=SceneTexture(UVs=@n45) {sceneTextureId=PPI_Roughness}.Color) {r=true}
  n61 = Add(A=Add(A=Subtract(A=@n56, B=@n57), B=Subtract(A=@n56, B=@n58)), B=Add(A=Subtract(A=@n56, B=@n59), B=Subtract(A=@n56, B=@n60)))
  n62 = ScalarParameter() {
   defaultValue=-0.04
   desc="a small value is picking up more roughness differences but get more noise.\r\na large value is not as detailed but has less noise."
   group=2 - ToonOutline_Roughness
   parameterName=Roughness_Divisor
   sliderMax=1
   sliderMin=-1
   sortPriority=11
  }
  n63 = Divide(A=AppendVector(A=@n61, B=@n61), B=AppendVector(A=@n62, B=Multiply(A=@n62) {constB=-1}))
  n64 = ScalarParameter() {
   defaultValue=4
   desc=a large value like 100 can improve roughness noise
   group=2 - ToonOutline_Roughness
   parameterName=Roughness_Scale
   sliderMax=100
   sortPriority=12
  }
  n65 = Multiply(A=OneMinus(None=Saturate(None=@n63)), B=@n64)
  n66 = MaterialFunctionCall(Float2=OneMinus(None=Saturate(None=@n65))) {materialFunction=&BreakOutFloat2Components}
  n67 = ScalarParameter() {
   desc=Used to draw outlines on both sides of a discontinuity
   group=2 - ToonOutline_Roughness
   parameterName=UseCenteredLines
   sliderMax=1
   sortPriority=13
  }
  n68 = Max(A=@n66.R, B=Multiply(A=@n66.G, B=@n67))
  n69 = StaticSwitchParameter(True=Multiply(A=@n18, B=@n68), False=@n68) {
   desc=Exclude a section of the mesh from having Distance Outlines based on a Texture
   group=2 - ToonOutline_Roughness
   parameterName=UseMaskTextureForRoughnessOutlines
   sortPriority=2
  }
  NamedRerouteDeclaration(None=@n69) {name=RoughnessOutlines, nodeColor=<rgba 0.2188 0 1 1>}
  n70 = Multiply(A=VertexNormalWS(), B=ScalarParameter() {group=3 - ToonOutline_WPO, parameterName=Mesh_Scale, sliderMax=1})
  NamedRerouteDeclaration(None=StaticSwitchParameter(
    True=MaterialFunctionCall!(WPO=@n70) {materialFunction=/Game/Packages/Fortress_SharedMaterials/Base_Material_Functions/Fortnite_HideElementsWithVertexColor}.Result
    False=@n70
   ) {group=3 - ToonOutline_Opacity, parameterName=Use Vertex Colors for Mask, sortPriority=1}) {name=WPO, nodeColor=<rgba 0 1 0.2734 1>}
  n71 = ScalarParameter() {group=FXSupport, parameterName=SpecialDeviceFix}
  n72 = Subtract(A=PixelDepth().R, B=SceneDepth().R)
  n73 = Multiply(A=ScalarParameter() {defaultValue=3, group=3 - ToonOutline_Opacity, parameterName=Outline_MaskBias, sliderMax=4}, B=NamedRerouteUsage())
  n74 = ComponentMask(None=NamedRerouteUsage()) {r=true}
  n75 = Max(A=NamedRerouteUsage(), B=@n74)
  n76 = Multiply(
   A=Saturate(None=Add(A=@n72, B=@n73))
   B=StaticSwitchParameter(True=ShadingPathSwitch(Default=@n75, Forward=@n74, Mobile=@n74), False=@n74) {defaultValue=true, group=2 - ToonOutline_Roughness, parameterName=UseRoughnessOutlines, sortPriority=1}
  )
  n77 = Multiply(A=NamedRerouteUsage(), B=@n76)
  n78 = Multiply(
   A=Saturate(None=StaticSwitchParameter(True=@n77, False=@n76) {group=ToonOutline_DontTouch, parameterName=UseSubsurfaceMasking, sortPriority=1})
   B=ScalarParameter() {defaultValue=1, group=3 - ToonOutline_Opacity, parameterName=Opacity_Scale, sliderMax=2}
  )
  NamedRerouteDeclaration(None=StaticSwitchParameter(True=LinearInterpolate(A=@n78, Alpha=@n71) {constB=0}, False=@n78) {group=FXSupport, parameterName=UseInNiagaraFX, sortPriority=1}) {name=Opacity, nodeColor=<rgba 1 0 0.4453 1>}
  n79 = TextureSampleParameter2D() {group=1 - ToonOutline, parameterName=ColorMap, sortPriority=2, texture=&T_BlacksRGB}
  n80 = ScalarParameter() {defaultValue=1, desc=Multiplier for the Outline Color Map, group=1 - ToonOutline, parameterName=Color_Intensity}
  n81 = MaterialFunctionCall!(In=StaticSwitchParameter(
    True=Multiply(A=@n79.RGB, B=@n80)
    False=VectorParameter() {group=1 - ToonOutline, parameterName=Color, sortPriority=3, defaultValue=<rgba 0 0 0 1>}.RGB
   ) {desc="Switch to Activate Color Customization ", group=1 - ToonOutline, parameterName=UseColorMap, sortPriority=1}) {materialFunction=/Game/Packages/Fortress_SharedMaterials/Anime_Master_Material/MF_TODCustomColors}
  NamedRerouteDeclaration(None=MaterialFunctionCall!(Color=MaterialFunctionCall!(In=@n81.Result, LUT, LUT Multiplier) {materialFunction=/Game/Packages/Fortress_SharedMaterials/Anime_Master_Material/MF_ApplyLUT}.Result) {materialFunction=/Game/Packages/Fortress_SharedMaterials/Anime_Master_Material/MF_TODColorAdjustment}.Result) {name=Emissive, nodeColor=<rgba 0.6641 1 0 1>}
  NamedRerouteDeclaration(None=Divide(A=ObjectRadius(), B=ScalarParameter() {defaultValue=97, group=3 - ToonOutline_WPO, parameterName=Object_Scale, sliderMax=200})) {name=ObjectRadius, nodeColor=<rgba 0 1 0.5547 1>}
  PinBase!(M=@n4, Pre Skinned Local Position=@n6, Distance Blend=@n7.Distance Blend, Fresnel=@n8) {reroutePins [
    M {expression=&self:MaterialExpressionReroute_8}
    Pre Skinned Local Position {expression=&self:MaterialExpressionReroute_9}
    Distance Blend {expression=&self:MaterialExpressionReroute_0}
    Fresnel {expression=&self:MaterialExpressionReroute_6}
   ]}
  PinBase!()
 ]
}
