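# Editor Utility Widget that batch-creates and reparents material instances for anime/toon characters.
# Buttons create a parent inverted-hull outline MI, then one MI plus outline MI per type-map entry, outlines using that parent or the selected outline material.
# Other buttons create outlines beside selected MIs or reparent them.
# The reparent dropdown lists MaterialArray, but Reparent indexes the type map's values, so choices can mismatch.
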EditorUtilityWidgetBlueprint /Game/ArtTools/Tech_Art/MaterialBatcher/BP_MaterialBatcher
graphs [
 EventGraph {nodes [
   "On Clicked (EditorUtilityButton_Reparent)"() {is=title K2Node_ComponentBoundEvent}
   > n1 = GetSelectedAssetsOfClass(AssetClass=/Script/Engine.MaterialInstanceConstant)
   > n2 = For Each Loop(Array:Array of Object References=@n1) -> (Array Element:Object Reference)
   > n3 = Cast To MaterialInstanceConstant(Object=@n2)
   > n4 = Values!(
    execute
    TargetMap:Map of Strings to Material Interface Object References=Get TypeArrayMapMaterial() -> (TypeArrayMapMaterial:Map of Strings to Material Interface Object References)
   ) -> (then, Values:Array of Material Interface Object References)
   n5 = GetSelectedIndex!(self:"Combo Box (String) Object Reference"=Get EditorUtilityComboBoxString_SelectedParents() -> (EditorUtilityComboBoxString_SelectedParents:Editor Utility Combo Box String Object Reference)) -> (ReturnValue:int=0)
   n6 = Length(TargetArray:Array of Material Interface Object References=@n4)
   > SetMaterialInstanceParent!(
    execute
    Instance:Material Instance Constant Object Reference=@n3
    NewParent:Material Interface Object Reference=Select(
     Option 0:Material Interface Object Reference="Get (a copy)"(Array:Array of Material Interface Object References=@n4, Dimension 1=@n5) -> (Output:Material Interface Object Reference)
     Option 1:Material Interface Object Reference=Get SelectedParent() -> (SelectedParent:Material Interface Object Reference)
     Index="Equal (Integer)"(A=@n5, B=@n6)
    ) -> (ReturnValue:Material Interface Object Reference)
   ) -> (then)
   > SaveLoadedAsset(AssetToSave=@n3, bOnlyIfIsDirty=false)
   "On Clicked (EditorUtilityButton_CreateParentMI)"() {is=title K2Node_ComponentBoundEvent}
   > n7 = GetAssetTools()
   > n8 = CreateAsset(
    self=@n7
    AssetName=Append!(A=MI_, B=Get AssetName() -> (AssetName:String), C:String=_InvHullOutline, D:String=_PARENT) -> (ReturnValue)
    PackagePath=Get Package Path() -> (Package Path:String)
    AssetClass=&MaterialInstanceConstant
   )
   > n9 = Cast To MaterialInstanceConstant(Object=@n8)
   > SetMaterialInstanceParent(Instance=@n9, NewParent=Get SelectedOutlineParent() -> (SelectedOutlineParent:Material Interface Object Reference))
   > SaveLoadedAsset(AssetToSave=@n9, bOnlyIfIsDirty=false)
   > Set ParentMI(ParentMI:Material Instance Constant Object Reference=@n9) -> (Output_Get:Material Instance Constant Object Reference)
   "On Clicked (EditorUtilityButton_CreateAllMIs)"() {is=title K2Node_ComponentBoundEvent}
   > n10 = Keys!(execute, TargetMap:Map of Strings to Material Interface Object References=Get TypeArrayMapMaterial()) -> (then, Keys:Array of Strings)
   > n11 = For Each Loop(Array:Array of Strings=@n10) -> (Array Element:String)
   > n12 = GetAssetTools()
   n13 = Append(A=MI_, B=Get AssetName(), C=_, D=@n11)
   n14 = Get Package Path()
   > n15 = CreateAsset(self=@n12, AssetName=@n13, PackagePath=@n14, AssetClass=&MaterialInstanceConstant)
   > n16 = Cast To MaterialInstanceConstant(Object=@n15)
   > n17 = Values(TargetMap=Get TypeArrayMapMaterial())
   > SetMaterialInstanceParent(Instance=@n16, NewParent="Get (a copy)"(Array=@n17, Dimension 1=@n11.Array Index))
   > SaveLoadedAsset(AssetToSave=@n16, bOnlyIfIsDirty=false)
   > n18 = GetAssetTools()
   > n19 = CreateAsset(
    self=@n18
    AssetName=Append!(A=@n13, B=_InvHullOutline) -> (ReturnValue)
    PackagePath=@n14
    AssetClass=&MaterialInstanceConstant
   )
   > n20 = Cast To MaterialInstanceConstant(Object=@n19)
   n21 = Get ParentMI() -> (ParentMI:Material Instance Constant Object Reference)
   n22 = Get Empty() -> (Empty:Object Reference)
   > SetMaterialInstanceParent(Instance=@n20, NewParent=Select(Option 0=@n21, Option 1=Get SelectedOutlineParent(), Index="Equal (Object)"(A=@n21, B=@n22)))
   > SaveLoadedAsset(AssetToSave=@n20, bOnlyIfIsDirty=false)
   "On Selection Changed (EditorUtilityComboBoxString_SelectedParents)"!() -> (OutputDelegate, then, SelectedItem:String, SelectionType:ESelectInfo Enum) {is=title K2Node_ComponentBoundEvent}
   n23 = Get MaterialArray() -> (MaterialArray:Array of Material Interface Object References)
   n24 = Get EditorUtilityComboBoxString_SelectedParents()
   > n25 = Branch(Condition="Equal (Integer)"(A=GetSelectedIndex(self=@n24) -> (ReturnValue=0), B=Length(TargetArray=@n23)))
   n26 = Get DetailsView_MaterialSelectionParent() -> (DetailsView_MaterialSelectionParent:Details View Object Reference)
   > SetIsEnabled(self=@n26, bInIsEnabled=true)
   SetIsEnabled(execute=@n25.else, self=@n26, bInIsEnabled=false)
   n27 = "On Check State Changed (EditorUtilityCheckBox_UseCustomParent)"!() -> (OutputDelegate, then, bIsChecked:bool) {is=title K2Node_ComponentBoundEvent}
   > n28 = Set UseCustomOutlineParent(UseCustomOutlineParent:bool=@n27.bIsChecked) -> (Output_Get:bool)
   > n29 = Branch(Condition=@n28)
   n30 = Get DetailsView_MaterialSelectionOutlineParent() -> (DetailsView_MaterialSelectionOutlineParent:Details View Object Reference)
   > SetIsEnabled(self=@n30, bInIsEnabled=true)
   Set SelectedOutlineParent(
    execute=@n29.else
    SelectedOutlineParent:Material Interface Object Reference=/Game/Global/Materials/Character_Parent_2023/Anime_Parent_2023/M_AnimeOutline_InvHull_2023
   ) -> (Output_Get:Material Interface Object Reference)
   > SetIsEnabled(self=@n30, bInIsEnabled=false)
   n31 = "On Text Changed (EditorUtilityEditableTextBox_Path)"!() -> (OutputDelegate, then, Text:"Text (by ref)") {is=title K2Node_ComponentBoundEvent}
   > Set Package Path(Package Path:String="To String (Text)"(InText=@n31.Text)) -> (Output_Get:String)
   n32 = "On Text Changed (EditorUtilityEditableTextBox_Name)"!() -> (OutputDelegate, then, Text:"Text (by ref)") {is=title K2Node_ComponentBoundEvent}
   > Set AssetName(AssetName:String="To String (Text)"(InText=@n32.Text)) -> (Output_Get:String)
   "On Clicked (EditorUtilityButton_CreateOutlineForSelection)"() {is=title K2Node_ComponentBoundEvent}
   > n33 = GetSelectedAssetsOfClass(AssetClass=&MaterialInstanceConstant)
   > n34 = For Each Loop(Array:Array of Object References=@n33) -> (Array Element:Object Reference)
   > n35 = GetAssetTools()
   n36 = GetDisplayName(Object=@n34)
   > n37 = CreateAsset(
    self=@n35
    AssetName=Append(A=@n36, B=_InvHullOutline)
    PackagePath=Replace(SourceString=Get Object Path String(Object=@n34), From=Append!(A=/, B=@n36, C:String=., D:String=@n36) -> (ReturnValue))
    AssetClass=&MaterialInstanceConstant
   )
   > n38 = Cast To MaterialInstanceConstant(Object=@n37)
   > SetMaterialInstanceParent(Instance=@n38, NewParent=Get SelectedOutlineParent())
   > SaveLoadedAsset(AssetToSave=@n38, bOnlyIfIsDirty=false)
   "On Clicked (EditorUtilityButton_GetCurrentPath)"() {is=title K2Node_ComponentBoundEvent}
   > Set Package Path(Package Path=GetCurrentContentBrowserPath!() -> (OutPath:String, ReturnValue:bool))
   Event Construct()
   > n39 = Sequence!(execute) -> (then_0, then_1, then_2:Exec)
   > n40 = For Each Loop(Array:Array of Material Interface Object References=Get MaterialArray()) -> (Array Element:Material Interface Object Reference)
   n41 = Get EditorUtilityComboBoxString_SelectedParents()
   AddOption!(execute=@n40.Completed, self:"Combo Box (String) Object Reference"=@n41, Option:String=Custom Parent Mateiral) -> (then)
   AddOption(execute=@n40.LoopBody, self=@n41, Option=GetDisplayName(Object=@n40))
   SetObject(
    execute=@n39.then_1
    self=[
     Get DetailsView_MaterialSelectionOutlineParent()
     Get DetailsView_CreateAllMIs() -> (DetailsView_CreateAllMIs:Details View Object Reference)
     Get DetailsView_MaterialSelectionParent()
    ]
    NewObject=Self-Reference()
   )
  ]}
 OnGenerateItemWidget {nodes [
   OnGenerateItemWidget!() -> (then, Item:Name) {is=title K2Node_FunctionEntry}
   > Return Node!(execute, ReturnValue:Widget Object Reference)
  ]}
 On_EditorUtilityComboBoxString_Material_GenerateWidget {nodes [
   On_EditorUtilityComboBoxString_Material_GenerateWidget!() -> (then, Item:String) {is=title K2Node_FunctionEntry}
   > Return Node!(execute, ReturnValue, ParentMI:Material Instance Constant Object Reference)
  ]}
]
variables [
 AssetName {type=string, default=AssetName}
 DirectoryPath {type=DirectoryPath, default {path=D:/P4/FN_PeterAnkermann_ReleaseC_MS/FortniteGame/Content/Developers/peterankermann/MateiralBatcher}}
 Empty {type=Object, default=null}
 MaterialArray {
  type="array<MaterialInterface>"
  default [
   /Game/Global/Materials/Character_Parent_2023/Anime_Parent_2023/MI_AnimeShader_2023_Default
   /Game/Global/Materials/Character_Parent_2023/Anime_Parent_2023/MI_AnimeShader_2023_Head
   /Game/Global/Materials/Character_Parent_2023/Anime_Parent_2023/MI_AnimeShader_2023_BackwardCompatible
   &M_AnimeOutline_InvHull_2023
   /Game/Global/Materials/Character_Parent_2023/Anime_Parent_2023/M_AnimeOutline_Overlap_2023_Fallback
  ]
 }
 Package Path {type=string, default=/Game/Developers/peterankermann/MateiralBatcher}
 ParentMI {type=MaterialInstanceConstant, default=null}
 SelectedOutlineParent {type=MaterialInterface, default=&M_AnimeOutline_InvHull_2023}
 SelectedParent {type=MaterialInterface, default=null}
 TypeArrayMapMaterial {
  type="map<MaterialInterface>"
  default {
   Body=&MI_AnimeShader_2023_Default
   FaceAcc=&MI_AnimeShader_2023_Default
   Head=&MI_AnimeShader_2023_Head
  }
 }
 UseCustomOutlineParent {type=boolean}
]
widgets {
 info {rootWidgetClass=&EditorUtilityScrollBox, widgetCount=33}
 widgets [
  EditorUtilityScrollBox_50 {widgetClassPath=&EditorUtilityScrollBox, widget=EditorUtilityScrollBox_50}
  VerticalBox_0 {widgetClassPath=&VerticalBox, parent=EditorUtilityScrollBox_50, widget=VerticalBox_0, slot {class=ScrollBoxSlot}}
  HorizontalBox_Name {widgetClassPath=&HorizontalBox, parent=VerticalBox_0, widget=HorizontalBox_Name, slot {class=VerticalBoxSlot}}
  RichTextBlock_AssetName {
   widgetClassPath=/Script/UMG.RichTextBlock
   bIsVariable=true
   parent=HorizontalBox_Name
   widget=RichTextBlock_AssetName
   properties {
    bOverrideDefaultStyle=true
    text="Asset Name: "
    defaultTextStyleOverride {
     overflowPolicy=Clip
     transformPolicy=None
     colorAndOpacity {colorUseRule=UseColor_Specified, specifiedColor=<rgba 0.8 0.8 0.8 1>}
     font {
      bForceMonospaced=false
      bMaterialIsStencil=false
      fontMaterial=null
      fontObject=/LauncherSocial/BrutalType
      letterSpacing=0
      monospacedWidth=1
      size=10
      skewAmount=0
      typefaceFontName=Regular
      outlineSettings {
       bApplyOutlineToDropShadows=false
       bMiteredCorners=false
       bSeparateFillAlpha=false
       outlineMaterial=null
       outlineSize=0
       outlineColor=<rgba 0 0 0 1>
      }
     }
     highlightColor {colorUseRule=UseColor_Specified, specifiedColor=<rgba 0 0 0 1>}
     highlightShape {
      bIsDynamicallyLoaded=false
      drawAs=Image
      imageType=NoImage
      mirroring=NoMirror
      resourceName=None
      resourceObject=null
      tiling=NoTile
      imageSize=<32 32>
      margin=<ltrb 0 0 0 0>
      outlineSettings {
       bUseBrushTransparency=false
       roundingType=HalfHeightRadius
       width=0
       color {colorUseRule=UseColor_Specified, specifiedColor=<rgba 0 0 0 0>}
       cornerRadii=<0 0 0 0>
      }
      tintColor {colorUseRule=UseColor_Specified, specifiedColor=<rgba 1 1 1 1>}
      uVRegion {bIsValid=false, max=<0 0>, min=<0 0>}
     }
     selectedBackgroundColor {colorUseRule=UseColor_Foreground, specifiedColor=<rgba 1 0 1 1>}
     shadowColorAndOpacity=<rgba 0 0 0 1>
     shadowOffset=<0 0>
     strikeBrush {
      bIsDynamicallyLoaded=false
      drawAs=Image
      imageType=NoImage
      mirroring=NoMirror
      resourceName=None
      resourceObject=null
      tiling=NoTile
      imageSize=<32 32>
      margin=<ltrb 0 0 0 0>
      outlineSettings {
       bUseBrushTransparency=false
       roundingType=HalfHeightRadius
       width=0
       color {colorUseRule=UseColor_Specified, specifiedColor=<rgba 0 0 0 0>}
       cornerRadii=<0 0 0 0>
      }
      tintColor {colorUseRule=UseColor_Specified, specifiedColor=<rgba 1 1 1 1>}
      uVRegion {bIsValid=false, max=<0 0>, min=<0 0>}
     }
     underlineBrush {
      bIsDynamicallyLoaded=false
      drawAs=Image
      imageType=NoImage
      mirroring=NoMirror
      resourceName=None
      resourceObject=null
      tiling=NoTile
      imageSize=<32 32>
      margin=<ltrb 0 0 0 0>
      outlineSettings {
       bUseBrushTransparency=false
       roundingType=HalfHeightRadius
       width=0
       color {colorUseRule=UseColor_Specified, specifiedColor=<rgba 0 0 0 0>}
       cornerRadii=<0 0 0 0>
      }
      tintColor {colorUseRule=UseColor_Specified, specifiedColor=<rgba 1 1 1 1>}
      uVRegion {bIsValid=false, max=<0 0>, min=<0 0>}
     }
    }
   }
   slot {class=HorizontalBoxSlot, properties {horizontalAlignment=HAlign_Center, verticalAlignment=VAlign_Center, padding=<ltrb 4 2 4 2>}}
  }
  EditorUtilityEditableTextBox_Name {
   widgetClassPath=&EditorUtilityEditableTextBox
   bIsVariable=true
   parent=HorizontalBox_Name
   widget=EditorUtilityEditableTextBox_Name
   slot {class=HorizontalBoxSlot, properties {padding=<ltrb 8 8 8 8>, size {sizeRule=Fill, value=1}}}
  }
  HorizontalBox_Path {widgetClassPath=&HorizontalBox, parent=VerticalBox_0, widget=HorizontalBox_Path, slot {class=VerticalBoxSlot}}
  RichTextBlock_AssetPath {
   widgetClassPath=&RichTextBlock
   bIsVariable=true
   parent=HorizontalBox_Path
   widget=RichTextBlock_AssetPath
   properties {
    bOverrideDefaultStyle=true
    text="Asset Path: "
    defaultTextStyleOverride {
     overflowPolicy=Clip
     transformPolicy=None
     colorAndOpacity {colorUseRule=UseColor_Specified, specifiedColor=<rgba 0.8 0.8 0.8 1>}
     font {
      bForceMonospaced=false
      bMaterialIsStencil=false
      fontMaterial=null
      fontObject=&BrutalType
      letterSpacing=0
      monospacedWidth=1
      size=10
      skewAmount=0
      typefaceFontName=Regular
      outlineSettings {
       bApplyOutlineToDropShadows=false
       bMiteredCorners=false
       bSeparateFillAlpha=false
       outlineMaterial=null
       outlineSize=0
       outlineColor=<rgba 0 0 0 1>
      }
     }
     highlightColor {colorUseRule=UseColor_Specified, specifiedColor=<rgba 0 0 0 1>}
     highlightShape {
      bIsDynamicallyLoaded=false
      drawAs=Image
      imageType=NoImage
      mirroring=NoMirror
      resourceName=None
      resourceObject=null
      tiling=NoTile
      imageSize=<32 32>
      margin=<ltrb 0 0 0 0>
      outlineSettings {
       bUseBrushTransparency=false
       roundingType=HalfHeightRadius
       width=0
       color {colorUseRule=UseColor_Specified, specifiedColor=<rgba 0 0 0 0>}
       cornerRadii=<0 0 0 0>
      }
      tintColor {colorUseRule=UseColor_Specified, specifiedColor=<rgba 1 1 1 1>}
      uVRegion {bIsValid=false, max=<0 0>, min=<0 0>}
     }
     selectedBackgroundColor {colorUseRule=UseColor_Foreground, specifiedColor=<rgba 1 0 1 1>}
     shadowColorAndOpacity=<rgba 0 0 0 1>
     shadowOffset=<0 0>
     strikeBrush {
      bIsDynamicallyLoaded=false
      drawAs=Image
      imageType=NoImage
      mirroring=NoMirror
      resourceName=None
      resourceObject=null
      tiling=NoTile
      imageSize=<32 32>
      margin=<ltrb 0 0 0 0>
      outlineSettings {
       bUseBrushTransparency=false
       roundingType=HalfHeightRadius
       width=0
       color {colorUseRule=UseColor_Specified, specifiedColor=<rgba 0 0 0 0>}
       cornerRadii=<0 0 0 0>
      }
      tintColor {colorUseRule=UseColor_Specified, specifiedColor=<rgba 1 1 1 1>}
      uVRegion {bIsValid=false, max=<0 0>, min=<0 0>}
     }
     underlineBrush {
      bIsDynamicallyLoaded=false
      drawAs=Image
      imageType=NoImage
      mirroring=NoMirror
      resourceName=None
      resourceObject=null
      tiling=NoTile
      imageSize=<32 32>
      margin=<ltrb 0 0 0 0>
      outlineSettings {
       bUseBrushTransparency=false
       roundingType=HalfHeightRadius
       width=0
       color {colorUseRule=UseColor_Specified, specifiedColor=<rgba 0 0 0 0>}
       cornerRadii=<0 0 0 0>
      }
      tintColor {colorUseRule=UseColor_Specified, specifiedColor=<rgba 1 1 1 1>}
      uVRegion {bIsValid=false, max=<0 0>, min=<0 0>}
     }
    }
   }
   slot {class=HorizontalBoxSlot, properties {horizontalAlignment=HAlign_Center, verticalAlignment=VAlign_Center, padding=<ltrb 4 2 4 2>}}
  }
  EditorUtilityEditableTextBox_Path {
   widgetClassPath=&EditorUtilityEditableTextBox
   bIsVariable=true
   parent=HorizontalBox_Path
   widget=EditorUtilityEditableTextBox_Path
   slot {class=HorizontalBoxSlot, properties {padding=<ltrb 8 8 8 8>, size {sizeRule=Fill, value=1}}}
  }
  EditorUtilityButton_GetCurrentPath {
   widgetClassPath=&EditorUtilityButton
   parent=HorizontalBox_Path
   widget=EditorUtilityButton_GetCurrentPath
   slot {class=HorizontalBoxSlot, properties {padding=<ltrb 8 8 8 8>}}
  }
  RichTextBlock_GetCurrentPath {
   widgetClassPath=&RichTextBlock
   bIsVariable=true
   parent=EditorUtilityButton_GetCurrentPath
   widget=RichTextBlock_GetCurrentPath
   properties {
    bOverrideDefaultStyle=true
    text=Get Path
    toolTipText=Get current Content Browser path
    defaultTextStyleOverride {
     overflowPolicy=Clip
     transformPolicy=None
     colorAndOpacity {colorUseRule=UseColor_Specified, specifiedColor=<rgba 0.8 0.8 0.8 1>}
     font {
      bForceMonospaced=false
      bMaterialIsStencil=false
      fontMaterial=null
      fontObject=&BrutalType
      letterSpacing=0
      monospacedWidth=1
      size=10
      skewAmount=0
      typefaceFontName=Regular
      outlineSettings {
       bApplyOutlineToDropShadows=false
       bMiteredCorners=false
       bSeparateFillAlpha=false
       outlineMaterial=null
       outlineSize=0
       outlineColor=<rgba 0 0 0 1>
      }
     }
     highlightColor {colorUseRule=UseColor_Specified, specifiedColor=<rgba 0 0 0 1>}
     highlightShape {
      bIsDynamicallyLoaded=false
      drawAs=Image
      imageType=NoImage
      mirroring=NoMirror
      resourceName=None
      resourceObject=null
      tiling=NoTile
      imageSize=<32 32>
      margin=<ltrb 0 0 0 0>
      outlineSettings {
       bUseBrushTransparency=false
       roundingType=HalfHeightRadius
       width=0
       color {colorUseRule=UseColor_Specified, specifiedColor=<rgba 0 0 0 0>}
       cornerRadii=<0 0 0 0>
      }
      tintColor {colorUseRule=UseColor_Specified, specifiedColor=<rgba 1 1 1 1>}
      uVRegion {bIsValid=false, max=<0 0>, min=<0 0>}
     }
     selectedBackgroundColor {colorUseRule=UseColor_Foreground, specifiedColor=<rgba 1 0 1 1>}
     shadowColorAndOpacity=<rgba 0 0 0 1>
     shadowOffset=<0 0>
     strikeBrush {
      bIsDynamicallyLoaded=false
      drawAs=Image
      imageType=NoImage
      mirroring=NoMirror
      resourceName=None
      resourceObject=null
      tiling=NoTile
      imageSize=<32 32>
      margin=<ltrb 0 0 0 0>
      outlineSettings {
       bUseBrushTransparency=false
       roundingType=HalfHeightRadius
       width=0
       color {colorUseRule=UseColor_Specified, specifiedColor=<rgba 0 0 0 0>}
       cornerRadii=<0 0 0 0>
      }
      tintColor {colorUseRule=UseColor_Specified, specifiedColor=<rgba 1 1 1 1>}
      uVRegion {bIsValid=false, max=<0 0>, min=<0 0>}
     }
     underlineBrush {
      bIsDynamicallyLoaded=false
      drawAs=Image
      imageType=NoImage
      mirroring=NoMirror
      resourceName=None
      resourceObject=null
      tiling=NoTile
      imageSize=<32 32>
      margin=<ltrb 0 0 0 0>
      outlineSettings {
       bUseBrushTransparency=false
       roundingType=HalfHeightRadius
       width=0
       color {colorUseRule=UseColor_Specified, specifiedColor=<rgba 0 0 0 0>}
       cornerRadii=<0 0 0 0>
      }
      tintColor {colorUseRule=UseColor_Specified, specifiedColor=<rgba 1 1 1 1>}
      uVRegion {bIsValid=false, max=<0 0>, min=<0 0>}
     }
    }
   }
   slot {class=ButtonSlot}
  }
  Spacer_1 {widgetClassPath=&Spacer, parent=VerticalBox_0, widget=Spacer_1, properties {size=<1 16>}, slot {class=VerticalBoxSlot}}
  HorizontalBox_Parent {widgetClassPath=&HorizontalBox, parent=VerticalBox_0, widget=HorizontalBox_Parent, slot {class=VerticalBoxSlot}}
  EditorUtilityButton_CreateParentMI {
   widgetClassPath=&EditorUtilityButton
   parent=HorizontalBox_Parent
   widget=EditorUtilityButton_CreateParentMI
   slot {class=HorizontalBoxSlot, properties {padding=<ltrb 8 8 8 8>}}
  }
  RichTextBlock_CreateParentMI {
   widgetClassPath=&RichTextBlock
   bIsVariable=true
   parent=EditorUtilityButton_CreateParentMI
   widget=RichTextBlock_CreateParentMI
   properties {
    bOverrideDefaultStyle=true
    text=1. Create Parent Outline MI
    defaultTextStyleOverride {
     overflowPolicy=Clip
     transformPolicy=None
     colorAndOpacity {colorUseRule=UseColor_Specified, specifiedColor=<rgba 0.8 0.8 0.8 1>}
     font {
      bForceMonospaced=false
      bMaterialIsStencil=false
      fontMaterial=null
      fontObject=&BrutalType
      letterSpacing=0
      monospacedWidth=1
      size=10
      skewAmount=0
      typefaceFontName=Regular
      outlineSettings {
       bApplyOutlineToDropShadows=false
       bMiteredCorners=false
       bSeparateFillAlpha=false
       outlineMaterial=null
       outlineSize=0
       outlineColor=<rgba 0 0 0 1>
      }
     }
     highlightColor {colorUseRule=UseColor_Specified, specifiedColor=<rgba 0 0 0 1>}
     highlightShape {
      bIsDynamicallyLoaded=false
      drawAs=Image
      imageType=NoImage
      mirroring=NoMirror
      resourceName=None
      resourceObject=null
      tiling=NoTile
      imageSize=<32 32>
      margin=<ltrb 0 0 0 0>
      outlineSettings {
       bUseBrushTransparency=false
       roundingType=HalfHeightRadius
       width=0
       color {colorUseRule=UseColor_Specified, specifiedColor=<rgba 0 0 0 0>}
       cornerRadii=<0 0 0 0>
      }
      tintColor {colorUseRule=UseColor_Specified, specifiedColor=<rgba 1 1 1 1>}
      uVRegion {bIsValid=false, max=<0 0>, min=<0 0>}
     }
     selectedBackgroundColor {colorUseRule=UseColor_Foreground, specifiedColor=<rgba 1 0 1 1>}
     shadowColorAndOpacity=<rgba 0 0 0 1>
     shadowOffset=<0 0>
     strikeBrush {
      bIsDynamicallyLoaded=false
      drawAs=Image
      imageType=NoImage
      mirroring=NoMirror
      resourceName=None
      resourceObject=null
      tiling=NoTile
      imageSize=<32 32>
      margin=<ltrb 0 0 0 0>
      outlineSettings {
       bUseBrushTransparency=false
       roundingType=HalfHeightRadius
       width=0
       color {colorUseRule=UseColor_Specified, specifiedColor=<rgba 0 0 0 0>}
       cornerRadii=<0 0 0 0>
      }
      tintColor {colorUseRule=UseColor_Specified, specifiedColor=<rgba 1 1 1 1>}
      uVRegion {bIsValid=false, max=<0 0>, min=<0 0>}
     }
     underlineBrush {
      bIsDynamicallyLoaded=false
      drawAs=Image
      imageType=NoImage
      mirroring=NoMirror
      resourceName=None
      resourceObject=null
      tiling=NoTile
      imageSize=<32 32>
      margin=<ltrb 0 0 0 0>
      outlineSettings {
       bUseBrushTransparency=false
       roundingType=HalfHeightRadius
       width=0
       color {colorUseRule=UseColor_Specified, specifiedColor=<rgba 0 0 0 0>}
       cornerRadii=<0 0 0 0>
      }
      tintColor {colorUseRule=UseColor_Specified, specifiedColor=<rgba 1 1 1 1>}
      uVRegion {bIsValid=false, max=<0 0>, min=<0 0>}
     }
    }
   }
   slot {class=ButtonSlot}
  }
  EditorUtilityCheckBox_UseCustomParent {
   widgetClassPath=&EditorUtilityCheckBox
   bIsVariable=true
   parent=HorizontalBox_Parent
   widget=EditorUtilityCheckBox_UseCustomParent
   slot {class=HorizontalBoxSlot}
  }
  RichTextBlock_UseCustomParent {
   widgetClassPath=&RichTextBlock
   bIsVariable=true
   parent=EditorUtilityCheckBox_UseCustomParent
   widget=RichTextBlock_UseCustomParent
   properties {
    bOverrideDefaultStyle=true
    text=Custom Parent Outline Mateiral
    defaultTextStyleOverride {
     overflowPolicy=Clip
     transformPolicy=None
     colorAndOpacity {colorUseRule=UseColor_Specified, specifiedColor=<rgba 0.8 0.8 0.8 1>}
     font {
      bForceMonospaced=false
      bMaterialIsStencil=false
      fontMaterial=null
      fontObject=&BrutalType
      letterSpacing=0
      monospacedWidth=1
      size=10
      skewAmount=0
      typefaceFontName=Regular
      outlineSettings {
       bApplyOutlineToDropShadows=false
       bMiteredCorners=false
       bSeparateFillAlpha=false
       outlineMaterial=null
       outlineSize=0
       outlineColor=<rgba 0 0 0 1>
      }
     }
     highlightColor {colorUseRule=UseColor_Specified, specifiedColor=<rgba 0 0 0 1>}
     highlightShape {
      bIsDynamicallyLoaded=false
      drawAs=Image
      imageType=NoImage
      mirroring=NoMirror
      resourceName=None
      resourceObject=null
      tiling=NoTile
      imageSize=<32 32>
      margin=<ltrb 0 0 0 0>
      outlineSettings {
       bUseBrushTransparency=false
       roundingType=HalfHeightRadius
       width=0
       color {colorUseRule=UseColor_Specified, specifiedColor=<rgba 0 0 0 0>}
       cornerRadii=<0 0 0 0>
      }
      tintColor {colorUseRule=UseColor_Specified, specifiedColor=<rgba 1 1 1 1>}
      uVRegion {bIsValid=false, max=<0 0>, min=<0 0>}
     }
     selectedBackgroundColor {colorUseRule=UseColor_Foreground, specifiedColor=<rgba 1 0 1 1>}
     shadowColorAndOpacity=<rgba 0 0 0 1>
     shadowOffset=<0 0>
     strikeBrush {
      bIsDynamicallyLoaded=false
      drawAs=Image
      imageType=NoImage
      mirroring=NoMirror
      resourceName=None
      resourceObject=null
      tiling=NoTile
      imageSize=<32 32>
      margin=<ltrb 0 0 0 0>
      outlineSettings {
       bUseBrushTransparency=false
       roundingType=HalfHeightRadius
       width=0
       color {colorUseRule=UseColor_Specified, specifiedColor=<rgba 0 0 0 0>}
       cornerRadii=<0 0 0 0>
      }
      tintColor {colorUseRule=UseColor_Specified, specifiedColor=<rgba 1 1 1 1>}
      uVRegion {bIsValid=false, max=<0 0>, min=<0 0>}
     }
     underlineBrush {
      bIsDynamicallyLoaded=false
      drawAs=Image
      imageType=NoImage
      mirroring=NoMirror
      resourceName=None
      resourceObject=null
      tiling=NoTile
      imageSize=<32 32>
      margin=<ltrb 0 0 0 0>
      outlineSettings {
       bUseBrushTransparency=false
       roundingType=HalfHeightRadius
       width=0
       color {colorUseRule=UseColor_Specified, specifiedColor=<rgba 0 0 0 0>}
       cornerRadii=<0 0 0 0>
      }
      tintColor {colorUseRule=UseColor_Specified, specifiedColor=<rgba 1 1 1 1>}
      uVRegion {bIsValid=false, max=<0 0>, min=<0 0>}
     }
    }
   }
   slot {class=PanelSlot}
  }
  DetailsView_MaterialSelectionOutlineParent {
   widgetClassPath=&DetailsView
   parent=HorizontalBox_Parent
   widget=DetailsView_MaterialSelectionOutlineParent
   properties {bIsEnabled=false, propertiesToShow [SelectedOutlineParent]}
   slot {class=HorizontalBoxSlot}
  }
  Spacer_97 {widgetClassPath=&Spacer, parent=VerticalBox_0, widget=Spacer_97, properties {size=<1 16>}, slot {class=VerticalBoxSlot}}
  HorizontalBox_90 {widgetClassPath=&HorizontalBox, parent=VerticalBox_0, widget=HorizontalBox_90, slot {class=VerticalBoxSlot}}
  EditorUtilityButton_CreateAllMIs {
   widgetClassPath=&EditorUtilityButton
   parent=HorizontalBox_90
   widget=EditorUtilityButton_CreateAllMIs
   slot {class=HorizontalBoxSlot, properties {padding=<ltrb 8 8 8 8>}}
  }
  RichTextBlock_CreateAllMIs {
   widgetClassPath=&RichTextBlock
   bIsVariable=true
   parent=EditorUtilityButton_CreateAllMIs
   widget=RichTextBlock_CreateAllMIs
   properties {
    bOverrideDefaultStyle=true
    text=2. Create all Toon MIs
    defaultTextStyleOverride {
     overflowPolicy=Clip
     transformPolicy=None
     colorAndOpacity {colorUseRule=UseColor_Specified, specifiedColor=<rgba 0.8 0.8 0.8 1>}
     font {
      bForceMonospaced=false
      bMaterialIsStencil=false
      fontMaterial=null
      fontObject=&BrutalType
      letterSpacing=0
      monospacedWidth=1
      size=10
      skewAmount=0
      typefaceFontName=Regular
      outlineSettings {
       bApplyOutlineToDropShadows=false
       bMiteredCorners=false
       bSeparateFillAlpha=false
       outlineMaterial=null
       outlineSize=0
       outlineColor=<rgba 0 0 0 1>
      }
     }
     highlightColor {colorUseRule=UseColor_Specified, specifiedColor=<rgba 0 0 0 1>}
     highlightShape {
      bIsDynamicallyLoaded=false
      drawAs=Image
      imageType=NoImage
      mirroring=NoMirror
      resourceName=None
      resourceObject=null
      tiling=NoTile
      imageSize=<32 32>
      margin=<ltrb 0 0 0 0>
      outlineSettings {
       bUseBrushTransparency=false
       roundingType=HalfHeightRadius
       width=0
       color {colorUseRule=UseColor_Specified, specifiedColor=<rgba 0 0 0 0>}
       cornerRadii=<0 0 0 0>
      }
      tintColor {colorUseRule=UseColor_Specified, specifiedColor=<rgba 1 1 1 1>}
      uVRegion {bIsValid=false, max=<0 0>, min=<0 0>}
     }
     selectedBackgroundColor {colorUseRule=UseColor_Foreground, specifiedColor=<rgba 1 0 1 1>}
     shadowColorAndOpacity=<rgba 0 0 0 1>
     shadowOffset=<0 0>
     strikeBrush {
      bIsDynamicallyLoaded=false
      drawAs=Image
      imageType=NoImage
      mirroring=NoMirror
      resourceName=None
      resourceObject=null
      tiling=NoTile
      imageSize=<32 32>
      margin=<ltrb 0 0 0 0>
      outlineSettings {
       bUseBrushTransparency=false
       roundingType=HalfHeightRadius
       width=0
       color {colorUseRule=UseColor_Specified, specifiedColor=<rgba 0 0 0 0>}
       cornerRadii=<0 0 0 0>
      }
      tintColor {colorUseRule=UseColor_Specified, specifiedColor=<rgba 1 1 1 1>}
      uVRegion {bIsValid=false, max=<0 0>, min=<0 0>}
     }
     underlineBrush {
      bIsDynamicallyLoaded=false
      drawAs=Image
      imageType=NoImage
      mirroring=NoMirror
      resourceName=None
      resourceObject=null
      tiling=NoTile
      imageSize=<32 32>
      margin=<ltrb 0 0 0 0>
      outlineSettings {
       bUseBrushTransparency=false
       roundingType=HalfHeightRadius
       width=0
       color {colorUseRule=UseColor_Specified, specifiedColor=<rgba 0 0 0 0>}
       cornerRadii=<0 0 0 0>
      }
      tintColor {colorUseRule=UseColor_Specified, specifiedColor=<rgba 1 1 1 1>}
      uVRegion {bIsValid=false, max=<0 0>, min=<0 0>}
     }
    }
   }
   slot {class=ButtonSlot}
  }
  DetailsView_CreateAllMIs {
   widgetClassPath=&DetailsView
   parent=HorizontalBox_90
   widget=DetailsView_CreateAllMIs
   properties {propertiesToShow [TypeArrayMapMaterial]}
   slot {class=HorizontalBoxSlot, properties {size {sizeRule=Fill, value=1}}}
  }
  Spacer {widgetClassPath=&Spacer, parent=VerticalBox_0, widget=Spacer, properties {size=<1 16>}, slot {class=VerticalBoxSlot}}
  HorizontalBox_Reparent {widgetClassPath=&HorizontalBox, parent=VerticalBox_0, widget=HorizontalBox_Reparent, slot {class=VerticalBoxSlot}}
  EditorUtilityButton_Reparent {
   widgetClassPath=&EditorUtilityButton
   parent=HorizontalBox_Reparent
   widget=EditorUtilityButton_Reparent
   slot {class=HorizontalBoxSlot, properties {padding=<ltrb 8 8 8 8>}}
  }
  RichTextBlock_Reparent {
   widgetClassPath=&RichTextBlock
   bIsVariable=true
   parent=EditorUtilityButton_Reparent
   widget=RichTextBlock_Reparent
   properties {
    bOverrideDefaultStyle=true
    text=3. Reparent
    defaultTextStyleOverride {
     overflowPolicy=Clip
     transformPolicy=None
     colorAndOpacity {colorUseRule=UseColor_Specified, specifiedColor=<rgba 0.8 0.8 0.8 1>}
     font {
      bForceMonospaced=false
      bMaterialIsStencil=false
      fontMaterial=null
      fontObject=&BrutalType
      letterSpacing=0
      monospacedWidth=1
      size=10
      skewAmount=0
      typefaceFontName=Regular
      outlineSettings {
       bApplyOutlineToDropShadows=false
       bMiteredCorners=false
       bSeparateFillAlpha=false
       outlineMaterial=null
       outlineSize=0
       outlineColor=<rgba 0 0 0 1>
      }
     }
     highlightColor {colorUseRule=UseColor_Specified, specifiedColor=<rgba 0 0 0 1>}
     highlightShape {
      bIsDynamicallyLoaded=false
      drawAs=Image
      imageType=NoImage
      mirroring=NoMirror
      resourceName=None
      resourceObject=null
      tiling=NoTile
      imageSize=<32 32>
      margin=<ltrb 0 0 0 0>
      outlineSettings {
       bUseBrushTransparency=false
       roundingType=HalfHeightRadius
       width=0
       color {colorUseRule=UseColor_Specified, specifiedColor=<rgba 0 0 0 0>}
       cornerRadii=<0 0 0 0>
      }
      tintColor {colorUseRule=UseColor_Specified, specifiedColor=<rgba 1 1 1 1>}
      uVRegion {bIsValid=false, max=<0 0>, min=<0 0>}
     }
     selectedBackgroundColor {colorUseRule=UseColor_Foreground, specifiedColor=<rgba 1 0 1 1>}
     shadowColorAndOpacity=<rgba 0 0 0 1>
     shadowOffset=<0 0>
     strikeBrush {
      bIsDynamicallyLoaded=false
      drawAs=Image
      imageType=NoImage
      mirroring=NoMirror
      resourceName=None
      resourceObject=null
      tiling=NoTile
      imageSize=<32 32>
      margin=<ltrb 0 0 0 0>
      outlineSettings {
       bUseBrushTransparency=false
       roundingType=HalfHeightRadius
       width=0
       color {colorUseRule=UseColor_Specified, specifiedColor=<rgba 0 0 0 0>}
       cornerRadii=<0 0 0 0>
      }
      tintColor {colorUseRule=UseColor_Specified, specifiedColor=<rgba 1 1 1 1>}
      uVRegion {bIsValid=false, max=<0 0>, min=<0 0>}
     }
     underlineBrush {
      bIsDynamicallyLoaded=false
      drawAs=Image
      imageType=NoImage
      mirroring=NoMirror
      resourceName=None
      resourceObject=null
      tiling=NoTile
      imageSize=<32 32>
      margin=<ltrb 0 0 0 0>
      outlineSettings {
       bUseBrushTransparency=false
       roundingType=HalfHeightRadius
       width=0
       color {colorUseRule=UseColor_Specified, specifiedColor=<rgba 0 0 0 0>}
       cornerRadii=<0 0 0 0>
      }
      tintColor {colorUseRule=UseColor_Specified, specifiedColor=<rgba 1 1 1 1>}
      uVRegion {bIsValid=false, max=<0 0>, min=<0 0>}
     }
    }
   }
   slot {class=ButtonSlot}
  }
  EditorUtilityComboBoxString_SelectedParents {
   widgetClassPath=/Script/Blutility.EditorUtilityComboBoxString
   bIsVariable=true
   parent=HorizontalBox_Reparent
   widget=EditorUtilityComboBoxString_SelectedParents
   properties {foregroundColor {colorUseRule=UseColor_Specified, specifiedColor=<rgba 0.5271 0.5271 0.5271 1>}}
   slot {class=HorizontalBoxSlot, properties {size {sizeRule=Fill, value=1}}}
  }
  DetailsView_MaterialSelectionParent {
   widgetClassPath=&DetailsView
   parent=HorizontalBox_Reparent
   widget=DetailsView_MaterialSelectionParent
   properties {bIsEnabled=false, propertiesToShow [SelectedParent]}
   slot {class=HorizontalBoxSlot}
  }
  Spacer_74 {widgetClassPath=&Spacer, parent=VerticalBox_0, widget=Spacer_74, properties {size=<1 24>}, slot {class=VerticalBoxSlot}}
  Border_197 {
   widgetClassPath=&Border
   parent=VerticalBox_0
   widget=Border_197
   properties {brushColor=<rgba 0 0 0 1>, padding=<ltrb 1 1 1 1>}
   slot {class=VerticalBoxSlot}
  }
  HorizontalBox_159 {widgetClassPath=&HorizontalBox, parent=VerticalBox_0, widget=HorizontalBox_159, slot {class=VerticalBoxSlot}}
  EditorUtilityButton_CreateOutlineForSelection {
   widgetClassPath=&EditorUtilityButton
   parent=HorizontalBox_159
   widget=EditorUtilityButton_CreateOutlineForSelection
   slot {class=HorizontalBoxSlot, properties {horizontalAlignment=HAlign_Center, padding=<ltrb 8 8 8 8>}}
  }
  RichTextBlock_CreateOutlineForSelection {
   widgetClassPath=&RichTextBlock
   bIsVariable=true
   parent=EditorUtilityButton_CreateOutlineForSelection
   widget=RichTextBlock_CreateOutlineForSelection
   properties {
    bOverrideDefaultStyle=true
    text=Create Outline Material for selection
    defaultTextStyleOverride {
     overflowPolicy=Clip
     transformPolicy=None
     colorAndOpacity {colorUseRule=UseColor_Specified, specifiedColor=<rgba 0.8 0.8 0.8 1>}
     font {
      bForceMonospaced=false
      bMaterialIsStencil=false
      fontMaterial=null
      fontObject=&BrutalType
      letterSpacing=0
      monospacedWidth=1
      size=10
      skewAmount=0
      typefaceFontName=Regular
      outlineSettings {
       bApplyOutlineToDropShadows=false
       bMiteredCorners=false
       bSeparateFillAlpha=false
       outlineMaterial=null
       outlineSize=0
       outlineColor=<rgba 0 0 0 1>
      }
     }
     highlightColor {colorUseRule=UseColor_Specified, specifiedColor=<rgba 0 0 0 1>}
     highlightShape {
      bIsDynamicallyLoaded=false
      drawAs=Image
      imageType=NoImage
      mirroring=NoMirror
      resourceName=None
      resourceObject=null
      tiling=NoTile
      imageSize=<32 32>
      margin=<ltrb 0 0 0 0>
      outlineSettings {
       bUseBrushTransparency=false
       roundingType=HalfHeightRadius
       width=0
       color {colorUseRule=UseColor_Specified, specifiedColor=<rgba 0 0 0 0>}
       cornerRadii=<0 0 0 0>
      }
      tintColor {colorUseRule=UseColor_Specified, specifiedColor=<rgba 1 1 1 1>}
      uVRegion {bIsValid=false, max=<0 0>, min=<0 0>}
     }
     selectedBackgroundColor {colorUseRule=UseColor_Foreground, specifiedColor=<rgba 1 0 1 1>}
     shadowColorAndOpacity=<rgba 0 0 0 1>
     shadowOffset=<0 0>
     strikeBrush {
      bIsDynamicallyLoaded=false
      drawAs=Image
      imageType=NoImage
      mirroring=NoMirror
      resourceName=None
      resourceObject=null
      tiling=NoTile
      imageSize=<32 32>
      margin=<ltrb 0 0 0 0>
      outlineSettings {
       bUseBrushTransparency=false
       roundingType=HalfHeightRadius
       width=0
       color {colorUseRule=UseColor_Specified, specifiedColor=<rgba 0 0 0 0>}
       cornerRadii=<0 0 0 0>
      }
      tintColor {colorUseRule=UseColor_Specified, specifiedColor=<rgba 1 1 1 1>}
      uVRegion {bIsValid=false, max=<0 0>, min=<0 0>}
     }
     underlineBrush {
      bIsDynamicallyLoaded=false
      drawAs=Image
      imageType=NoImage
      mirroring=NoMirror
      resourceName=None
      resourceObject=null
      tiling=NoTile
      imageSize=<32 32>
      margin=<ltrb 0 0 0 0>
      outlineSettings {
       bUseBrushTransparency=false
       roundingType=HalfHeightRadius
       width=0
       color {colorUseRule=UseColor_Specified, specifiedColor=<rgba 0 0 0 0>}
       cornerRadii=<0 0 0 0>
      }
      tintColor {colorUseRule=UseColor_Specified, specifiedColor=<rgba 1 1 1 1>}
      uVRegion {bIsValid=false, max=<0 0>, min=<0 0>}
     }
    }
   }
   slot {class=ButtonSlot}
  }
 ]
}
